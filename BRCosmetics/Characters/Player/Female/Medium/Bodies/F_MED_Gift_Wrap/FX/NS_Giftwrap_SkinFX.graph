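# Niagara skin effect for the female Gift Wrap character body, shedding small pale-blue sprites from skeletal mesh bones or sockets.
# Persistent emits a random trickle, while otherwise identical Landed scales its spawn rate by System Timeline normalized time.
# User hand/foot hidden flags choose the random bone/socket index range.
# Particles scatter in a cylinder, inherit some velocity, then drift down with noise, spinning and fading in and out.
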NiagaraSystem /BRCosmetics/Characters/Player/Female/Medium/Bodies/F_MED_Gift_Wrap/FX/NS_Giftwrap_SkinFX
dependencies {
 usedDataInterfaces [&NiagaraDataInterfaceCurve, &NiagaraDataInterfaceSkeletalMesh]
 usedDynamicInputs []
 usedModules [
  &self:SystemSpawnScript.SystemScriptSource.SystemScriptGraph.NiagaraNodeAssignment_0.SetVariables_#1
  &SystemState
  &Timeline
 ]
}
emitters {
 Landed {
  emitterUpdateScript [
   &EmitterState()
   &SpawnRate(SpawnRate=&Multiply_Float(A=100, B=$System.Timeline.NormalizedTime), Use Spawn Probability=false)
  ]
  particleSpawnScript [
   &InitializeParticle(
    Color=<rgba 1.1719 1.8412 3 1>
    Lifetime Max=1
    Lifetime Min=0.35
    Lifetime Mode=::NewEnumerator1
    Sprite Rotation Mode=::NewEnumerator2
    Sprite Size Mode=::NewEnumerator1
    Uniform Sprite Size Max=4
    Uniform Sprite Size Min=3
    UsePositionOffset=false
    Write Material Random=true
   )
   &SkeletalMeshLocation(
    Bone / Socket Index=&MakeCustomIntFromBool(
     Bool=$User.bisFootHidden
     True INT=&MakeCustomIntFromBool(Bool=$User.bisHandHidden, True INT=0, False INT=&RandomRangeInt(Minimum=4, Maximum=5))
     False INT=&MakeCustomIntFromBool(
      Bool=$User.bisHandHidden
      True INT=&RandomRangeInt(Minimum=0, Maximum=3)
      False INT=&RandomRangeInt(Minimum=0, Maximum=5)
     )
    )
    Bone Sampling Mode=::NewEnumerator4
    MaskSpawnGroup=false
    Skeletal Mesh=$System.Skeletal Mesh
   )
   &ShapeLocation(
    Cylinder Height=10
    Cylinder Radius=15
    MaskSpawnGroup=false
    Override Seed=false
    Rotation Axis=<0 0 1>
    Rotation Mode=::NewEnumerator1
    Shape Primitive=::NewEnumerator1
    Surface Only=false
    UseOwnerScale=true
   )
   &CameraOffset(Camera Offset Amount=5)
   &InheritVelocity(
    Inherited Velocity Amount Scale=&RandomRangeVector(Minimum=<0.11 0.11 0.11>, Maximum=<0.25 0.25 0.25>)
    Speed Limit=true
    UseSpeedThreshold=false
   )
  ]
  particleUpdateScript [
   &ParticleState()
   &SpriteRotationRate(Rotation Rate=&RandomRangeFloat(Minimum=-0.5, Maximum=0.5))
   &AccelerationForce(Acceleration=<0 0 -50>)
   &VectorNoiseForce(Force Amount=50, Module.Use Random Noise Time Interval=true, Random Noise Time Interval=0.05)
   &SolveForcesAndVelocity(Clamp Velocity=false, Limit Acceleration=false)
   &ScaleColor(
    Scale Alpha=&FloatFromCurve(FloatCurve=DataInterface() {propertyValues {Curve {keys [<at 0 0 RCIM_Cubic>,<at 0.2 1 RCIM_Cubic>,<at 1 0 RCIM_Cubic>]}}})
    ScaleA=true
    ScaleRGB=true
   )
  ]
  rendererClasses [&NiagaraSpriteRendererProperties]
  renderers [{rendererClass=&NiagaraSpriteRendererProperties}]
 }
 Persistent {
  emitterUpdateScript [
   &EmitterState()
   &SpawnRate(SpawnRate=&RandomRangeFloat(Minimum=0, Maximum=8), Use Spawn Probability=false)
  ]
  particleSpawnScript [
   &InitializeParticle(
    Color=<rgba 1.1719 1.8412 3 1>
    Lifetime Max=1
    Lifetime Min=0.35
    Lifetime Mode=::NewEnumerator1
    Sprite Rotation Mode=::NewEnumerator2
    Sprite Size Mode=::NewEnumerator1
    Uniform Sprite Size Max=4
    Uniform Sprite Size Min=3
    UsePositionOffset=false
    Write Material Random=true
   )
   &SkeletalMeshLocation(
    Bone / Socket Index=&MakeCustomIntFromBool(
     Bool=$User.bisFootHidden
     True INT=&MakeCustomIntFromBool(Bool=$User.bisHandHidden, True INT=0, False INT=&RandomRangeInt(Minimum=4, Maximum=5))
     False INT=&MakeCustomIntFromBool(
      Bool=$User.bisHandHidden
      True INT=&RandomRangeInt(Minimum=0, Maximum=3)
      False INT=&RandomRangeInt(Minimum=0, Maximum=5)
     )
    )
    Bone Sampling Mode=::NewEnumerator4
    MaskSpawnGroup=false
    Skeletal Mesh=$System.Skeletal Mesh
   )
   &ShapeLocation(
    Cylinder Height=10
    Cylinder Radius=15
    MaskSpawnGroup=false
    Override Seed=false
    Rotation Axis=<0 0 1>
    Rotation Mode=::NewEnumerator1
    Shape Primitive=::NewEnumerator1
    Surface Only=false
    UseOwnerScale=true
   )
   &CameraOffset(Camera Offset Amount=5)
   &InheritVelocity(
    Inherited Velocity Amount Scale=&RandomRangeVector(Minimum=<0.11 0.11 0.11>, Maximum=<0.25 0.25 0.25>)
    Speed Limit=true
    UseSpeedThreshold=false
   )
  ]
  particleUpdateScript [
   &ParticleState()
   &SpriteRotationRate(Rotation Rate=&RandomRangeFloat(Minimum=-0.5, Maximum=0.5))
   &AccelerationForce(Acceleration=<0 0 -50>)
   &VectorNoiseForce(Force Amount=50, Module.Use Random Noise Time Interval=true, Random Noise Time Interval=0.05)
   &SolveForcesAndVelocity(Clamp Velocity=false, Limit Acceleration=false)
   &ScaleColor(
    Scale Alpha=&FloatFromCurve(FloatCurve=DataInterface() {propertyValues {Curve {keys [<at 0 0 RCIM_Cubic>,<at 0.2 1 RCIM_Cubic>,<at 1 0 RCIM_Cubic>]}}})
    ScaleA=true
    ScaleRGB=true
   )
  ]
  rendererClasses [&NiagaraSpriteRendererProperties]
  renderers [{rendererClass=&NiagaraSpriteRendererProperties}]
 }
}
summary {
 systemName=NS_Giftwrap_SkinFX
 emitters [
  {emitterName=Persistent, rendererClasses [&NiagaraSpriteRendererProperties]}
  {emitterName=Landed, rendererClasses [&NiagaraSpriteRendererProperties]}
 ]
 userVariables [
  User.bhasLanded {type=NiagaraBool}
  User.bisFootHidden {type=NiagaraBool}
  User.bisHandHidden {type=NiagaraBool}
 ]
}
