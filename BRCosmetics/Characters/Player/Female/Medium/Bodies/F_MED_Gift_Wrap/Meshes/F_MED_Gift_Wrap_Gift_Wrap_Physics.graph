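# Physics asset for the female medium Gift Wrap body cosmetic, setting up a ribbon on dyn_ bones along with pelvis and thigh bodies.
# The main ribbon box hangs off a small base capsule on the pelvis and carries small base capsules for the side pieces.
# Base joints and thighs have free rotation, while ribbon segments use angle limits.
# Two-segment A–B chains are mirrored; of the sub-ribbons, only left subA has a box beyond its base.
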PhysicsAsset /BRCosmetics/Characters/Player/Female/Medium/Bodies/F_MED_Gift_Wrap/Meshes/F_MED_Gift_Wrap_Gift_Wrap_Physics
bodies {
 dyn_ribbon_a_base_l [{shapeType=Capsule, length=0.5, radius=0.5}]
 dyn_ribbon_a_base_r [{shapeType=Capsule, length=0.5, radius=0.5}]
 dyn_ribbon_a_l [{shapeType=Box, extentX=4.5, extentY=1.5, extentZ=5.5, center=<2.5383 0.0001 0>}]
 dyn_ribbon_a_r [{shapeType=Box, extentX=4.5, extentY=1.5, extentZ=6.5, center=<-2.5383 -0.0001 0>, rotation=<rot 0 0 180>}]
 dyn_ribbon_b_l [{shapeType=Box, extentX=11, extentY=1.5, extentZ=8, center=<5.6844 0.0001 0>}]
 dyn_ribbon_b_r [{shapeType=Box, extentX=11, extentY=1.5, extentZ=8, center=<-5.6844 -0.0001 0>, rotation=<rot 0 0 180>}]
 dyn_ribbon_main [{shapeType=Box, extentX=4.5, extentY=3.5, extentZ=5, center=<1.7301 0 0>}]
 dyn_ribbon_main_base [{shapeType=Capsule, length=0.5, radius=0.5}]
 dyn_ribbon_subA_base_l [{shapeType=Capsule, length=0.5, radius=0.5}]
 dyn_ribbon_subA_base_r [{shapeType=Capsule, length=0.5, radius=0.5}]
 dyn_ribbon_subA_l [{shapeType=Box, extentX=5.5, extentY=3.5, extentZ=11.5, center=<0.0001 0 4.6622>}]
 dyn_ribbon_subB_base_l [{shapeType=Capsule, length=0.5, radius=0.5}]
 dyn_ribbon_subB_base_r [{shapeType=Capsule, length=0.5, radius=0.5}]
 pelvis [{shapeType=Capsule, length=8.75, radius=11.75, center=<-3.9872 -0.3753 0>}]
 thigh_l [{shapeType=Capsule, length=26, radius=8.25, center=<-23.3008 0.0001 0>, rotation=<rot 90 0 0>}]
 thigh_r [{shapeType=Capsule, length=26, radius=8.25, center=<23.3008 -0.0001 0>, rotation=<rot 90 0 180>}]
}
constraints [
 {
  bone1Name=dyn_ribbon_a_base_l
  bone2Name=dyn_ribbon_main
  swing1LimitDegrees=45
  swing1Motion=Free
  swing2LimitDegrees=45
  swing2Motion=Free
  twistLimitDegrees=45
  twistMotion=Free
 }
 {
  bone1Name=dyn_ribbon_subB_base_l
  bone2Name=dyn_ribbon_main
  swing1LimitDegrees=45
  swing1Motion=Free
  swing2LimitDegrees=45
  swing2Motion=Free
  twistLimitDegrees=45
  twistMotion=Free
 }
 {
  bone1Name=dyn_ribbon_subA_base_l
  bone2Name=dyn_ribbon_main
  swing1LimitDegrees=45
  swing1Motion=Free
  swing2LimitDegrees=45
  swing2Motion=Free
  twistLimitDegrees=45
  twistMotion=Free
 }
 {
  bone1Name=dyn_ribbon_a_base_r
  bone2Name=dyn_ribbon_main
  swing1LimitDegrees=45
  swing1Motion=Free
  swing2LimitDegrees=45
  swing2Motion=Free
  twistLimitDegrees=45
  twistMotion=Free
 }
 {
  bone1Name=dyn_ribbon_subB_base_r
  bone2Name=dyn_ribbon_main
  swing1LimitDegrees=45
  swing1Motion=Free
  swing2LimitDegrees=45
  swing2Motion=Free
  twistLimitDegrees=45
  twistMotion=Free
 }
 {
  bone1Name=dyn_ribbon_subA_base_r
  bone2Name=dyn_ribbon_main
  swing1LimitDegrees=45
  swing1Motion=Free
  swing2LimitDegrees=45
  swing2Motion=Free
  twistLimitDegrees=45
  twistMotion=Free
 }
 {bone1Name=dyn_ribbon_main, bone2Name=dyn_ribbon_main_base, swing1LimitDegrees=15, swing2LimitDegrees=15, twistLimitDegrees=10}
 {bone1Name=dyn_ribbon_a_l, bone2Name=dyn_ribbon_a_base_l, swing1LimitDegrees=45, swing2LimitDegrees=15, twistLimitDegrees=5}
 {bone1Name=dyn_ribbon_b_l, bone2Name=dyn_ribbon_a_l, swing1LimitDegrees=45, swing2LimitDegrees=15, twistLimitDegrees=5}
 {
  bone1Name=dyn_ribbon_main_base
  bone2Name=pelvis
  swing1LimitDegrees=45
  swing1Motion=Free
  swing2LimitDegrees=45
  swing2Motion=Free
  twistLimitDegrees=45
  twistMotion=Free
 }
 {
  bone1Name=thigh_l
  bone2Name=pelvis
  swing1LimitDegrees=45
  swing1Motion=Free
  swing2LimitDegrees=45
  swing2Motion=Free
  twistLimitDegrees=45
  twistMotion=Free
 }
 {
  bone1Name=thigh_r
  bone2Name=pelvis
  swing1LimitDegrees=45
  swing1Motion=Free
  swing2LimitDegrees=45
  swing2Motion=Free
  twistLimitDegrees=45
  twistMotion=Free
 }
 {bone1Name=dyn_ribbon_a_r, bone2Name=dyn_ribbon_a_base_r, swing1LimitDegrees=45, swing2LimitDegrees=15, twistLimitDegrees=5}
 {bone1Name=dyn_ribbon_b_r, bone2Name=dyn_ribbon_a_r, swing1LimitDegrees=45, swing2LimitDegrees=15, twistLimitDegrees=5}
 {bone1Name=dyn_ribbon_subA_l, bone2Name=dyn_ribbon_subA_base_l, swing1LimitDegrees=5, swing2LimitDegrees=5, twistLimitDegrees=20}
]
