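# Physics asset for the face accessory part of the female medium Gift Wrap character, defining dynamic hair physics.
# Capsules on upper spine, neck and head form the body, joined by free angular constraints.
# Front pieces and two point bones constrain directly to the head, while three two-segment box chains attach through small, freely rotating base capsules.
# Chain segments have tight twist limits and wider swing, and three back points branch from the A and B chain tips.
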PhysicsAsset /BRCosmetics/Characters/Player/Female/Medium/Bodies/F_MED_Gift_Wrap/Meshes/Parts/F_MED_Gift_Wrap_FaceAcc_Physics
bodies {
 dyn_hair_back_pointA [{shapeType=Capsule, length=1.5, radius=1, center=<1.49 0 0>, rotation=<rot 90 0 0>}]
 dyn_hair_back_pointB [{shapeType=Capsule, length=0.75, radius=0.75, center=<0.9991 0 0>, rotation=<rot 90 0 0>}]
 dyn_hair_back_pointC [{shapeType=Capsule, length=1, radius=1, center=<1.4383 0.0001 0>, rotation=<rot 90 0 0>}]
 dyn_hair_back_pointD [{shapeType=Capsule, length=1.5, radius=1.25, center=<1.8172 -0.0001 0>, rotation=<rot 90 0 0>}]
 dyn_hair_front_A [{shapeType=Box, extentX=4.5, extentY=1.5, extentZ=4.5, center=<2.4473 0.7623 0>}]
 dyn_hair_front_B [{shapeType=Box, extentX=4.5, extentY=1.5, extentZ=4, center=<2.3208 1.0295 0>}]
 dyn_hair_front_pointA [{shapeType=Capsule, length=1, radius=1, center=<1.4346 0 0>, rotation=<rot 90 0 0>}]
 dyn_hair_subA_a [{shapeType=Box, extentX=5.5, extentY=5.5, extentZ=1.5, center=<2.8927 0 -0.3209>, rotation=<rot -1.5955 0 0>}]
 dyn_hair_subA_a_base [{shapeType=Capsule, length=0.5, radius=0.5}]
 dyn_hair_subA_b [{shapeType=Box, extentX=5.5, extentY=4, extentZ=1.5, center=<1.8049 0 -0.418>}]
 dyn_hair_subB_a [{shapeType=Box, extentX=6.5, extentY=6.5, extentZ=2.5, center=<3.249 0 0.215>}]
 dyn_hair_subB_a_base [{shapeType=Capsule, length=0.5, radius=0.5}]
 dyn_hair_subB_b [{shapeType=Box, extentX=5.5, extentY=6.5, extentZ=2, center=<2.5914 0 -0.0678>, rotation=<rot 0.7437 0 0>}]
 dyn_hair_subC_a [{shapeType=Box, extentX=6.5, extentY=2.5, extentZ=6, center=<3.3591 0.1687 -0.0001>}]
 dyn_hair_subC_a_base [{shapeType=Capsule, length=0.5, radius=0.5}]
 dyn_hair_subC_b [{shapeType=Box, extentX=5.5, extentY=1.5, extentZ=6, center=<2.5592 -0.3807 0>}]
 head [{shapeType=Capsule, length=6.5, radius=6.75, center=<3.2947 1.904 0.0001>, rotation=<rot 90 0 0>}]
 neck_01 [{shapeType=Capsule, length=7.5, radius=4, center=<0 0.2798 0>, rotation=<rot 90 0 0>}]
 spine_05 [{shapeType=Capsule, length=8.505, radius=8.755, center=<0.774 0.315 0>, rotation=<rot 90 0 -5.4357>}]
}
constraints [
 {
  bone1Name=head
  bone2Name=neck_01
  swing1LimitDegrees=45
  swing1Motion=Free
  swing2LimitDegrees=45
  swing2Motion=Free
  twistLimitDegrees=45
  twistMotion=Free
 }
 {
  bone1Name=neck_01
  bone2Name=spine_05
  swing1LimitDegrees=45
  swing1Motion=Free
  swing2LimitDegrees=45
  swing2Motion=Free
  twistLimitDegrees=45
  twistMotion=Free
 }
 {bone1Name=dyn_hair_front_A, bone2Name=head, swing1LimitDegrees=40, swing2LimitDegrees=5, twistLimitDegrees=5}
 {bone1Name=dyn_hair_front_B, bone2Name=head, swing1LimitDegrees=40, swing2LimitDegrees=5, twistLimitDegrees=5}
 {bone1Name=dyn_hair_back_pointA, bone2Name=head, swing1LimitDegrees=20, swing2LimitDegrees=20, twistLimitDegrees=5}
 {bone1Name=dyn_hair_front_pointA, bone2Name=head, swing1LimitDegrees=20, swing2LimitDegrees=20, twistLimitDegrees=5}
 {
  bone1Name=dyn_hair_subA_a_base
  bone2Name=head
  swing1LimitDegrees=45
  swing1Motion=Free
  swing2LimitDegrees=45
  swing2Motion=Free
  twistLimitDegrees=45
  twistMotion=Free
 }
 {
  bone1Name=dyn_hair_subB_a_base
  bone2Name=head
  swing1LimitDegrees=45
  swing1Motion=Free
  swing2LimitDegrees=45
  swing2Motion=Free
  twistLimitDegrees=45
  twistMotion=Free
 }
 {
  bone1Name=dyn_hair_subC_a_base
  bone2Name=head
  swing1LimitDegrees=45
  swing1Motion=Free
  swing2LimitDegrees=45
  swing2Motion=Free
  twistLimitDegrees=45
  twistMotion=Free
 }
 {bone1Name=dyn_hair_subA_a, bone2Name=dyn_hair_subA_a_base, swing1LimitDegrees=20, swing2LimitDegrees=30, twistLimitDegrees=5}
 {bone1Name=dyn_hair_subA_b, bone2Name=dyn_hair_subA_a, swing1LimitDegrees=20, swing2LimitDegrees=30, twistLimitDegrees=5}
 {bone1Name=dyn_hair_subB_a, bone2Name=dyn_hair_subB_a_base, swing1LimitDegrees=20, swing2LimitDegrees=30, twistLimitDegrees=5}
 {bone1Name=dyn_hair_subB_b, bone2Name=dyn_hair_subB_a, swing1LimitDegrees=20, swing2LimitDegrees=30, twistLimitDegrees=5}
 {bone1Name=dyn_hair_subC_a, bone2Name=dyn_hair_subC_a_base, swing1LimitDegrees=30, swing2LimitDegrees=20, twistLimitDegrees=5}
 {bone1Name=dyn_hair_subC_b, bone2Name=dyn_hair_subC_a, swing1LimitDegrees=30, swing2LimitDegrees=20, twistLimitDegrees=5}
 {bone1Name=dyn_hair_back_pointC, bone2Name=dyn_hair_subB_b, swing1LimitDegrees=25, swing2LimitDegrees=25, twistLimitDegrees=5}
 {bone1Name=dyn_hair_back_pointB, bone2Name=dyn_hair_subB_b, swing1LimitDegrees=25, swing2LimitDegrees=25, twistLimitDegrees=5}
 {bone1Name=dyn_hair_back_pointD, bone2Name=dyn_hair_subA_b, swing1LimitDegrees=25, swing2LimitDegrees=25, twistLimitDegrees=5}
]
