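# Post-process style animation Blueprint for the face accessory part of the female Gift Wrap body, adding secondary hair motion.
# Copies the attached parent's pose, then constrains the hairpin bone to weighted hair and head bones, mostly dyn_hair_subA_a.
# A RigidBody node then simulates the override physics asset relative to spine_05, with damped component motion and overridden gravity.
# Two additive bone-space rotations finally nudge hair bones; the event graph has no logic.
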AnimBlueprint /BRCosmetics/Characters/Player/Female/Medium/Bodies/F_MED_Gift_Wrap/Meshes/Parts/F_MED_Gift_Wrap_FaceAcc_AnimBP
target_skeleton=/Game/Characters/Player/Female/Medium/Bodies/F_MED_Gift_Wrap/Meshes/Parts/F_MED_Gift_Wrap_FaceAcc_Skeleton
graphs [
 AnimGraph {nodes [
   n1 = "Constraint - dyn_hairpin (PPPPP)"!(
    ComponentPose=Local To Component(LocalPose=Copy Pose From Mesh() {bUseAttachedParent=true})
    Alpha=1
    ConstraintWeights_0=0.0236
    ConstraintWeights_1:float=0.0102
    ConstraintWeights_2:float=0.0177
    ConstraintWeights_3:float=0.139
    ConstraintWeights_4:float=0.8095
   ) -> (Pose) {
    alphaBoolBlend {blendOption=Linear}
    boneToModify {boneName=dyn_hairpin}
    constraintSetup [
     {targetBone {boneName=dyn_hair_subB_a}}
     {targetBone {boneName=head}}
     {targetBone {boneName=dyn_hair_back_pointD}}
     {targetBone {boneName=dyn_hair_subA_b}}
     {targetBone {boneName=dyn_hair_subA_a}}
    ]
    constraintWeights [1,1,1,1,1]
   }
   n2 = RigidBody(ComponentPose=@n1) {
    bClampLinearTranslationLimitToRefPose=true
    bOverrideWorldGravity=true
    overridePhysicsAsset=&/F_MED_Gift_Wrap_FaceAcc_Physics
    simulationSpace=BaseBoneSpace
    alphaBoolBlend {blendOption=Linear}
    baseBoneRef {boneName=spine_05}
    componentLinearAccScale=<0.35 0.35 0.35>
    componentLinearVelScale=<0.35 0.35 0.35>
    overrideWorldGravity {z=-980}
   }
   n3 = "Transform (Modify) Bone - Bone: dyn_hair_subA_a"(ComponentPose=@n2, Alpha=1, Translation="0,0,0", Rotation="-3,0,0", Scale="1,1,1") {
    rotationMode=BMM_Additive
    rotationSpace=BCS_BoneSpace
    scaleMode=BMM_Additive
    scaleSpace=BCS_BoneSpace
    translationMode=BMM_Additive
    translationSpace=BCS_BoneSpace
    alphaBoolBlend {blendOption=Linear}
    boneToModify {boneName=dyn_hair_subA_a}
    rotation {pitch=-3}
   }
   n4 = "Transform (Modify) Bone - Bone: dyn_hair_subB_a_base"(ComponentPose=@n3, Alpha=1, Translation="0,0,0", Rotation="-3,0,0", Scale="1,1,1") {
    rotationMode=BMM_Additive
    rotationSpace=BCS_BoneSpace
    scaleMode=BMM_Additive
    scaleSpace=BCS_BoneSpace
    translationMode=BMM_Additive
    translationSpace=BCS_BoneSpace
    alphaBoolBlend {blendOption=Linear}
    boneToModify {boneName=dyn_hair_subB_a_base}
    rotation {pitch=-3}
   }
   Output Pose(Result=Component To Local(ComponentPose=@n4)) {layerGroup=None}
  ]}
 EventGraph {nodes [
   TryGetPawnOwner()
   Event BlueprintUpdateAnimation()
  ]}
]
variables []
